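# Physics asset for the Moss CarbideBlack pickaxe cosmetic, giving physics bodies to the vines on its head, shaft and base.
# One long capsule covers the handle, and each of the thirteen vines has a small MSTR anchor capsule constrained to it with free swing and twist.
# Chains of one to three capsules extend from each anchor with per-joint angle settings, and the two-segment head vines set swing to free.
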PhysicsAsset /BRCosmetics/Weapons/FORT_Melee/Pickaxe_Moss_CarbideBlack/Meshes/Moss_CarbideBlack_Physics
bodies {
 Melee_TwoHanded_Handle [{shapeType=Capsule, length=100, radius=3, center=<0 -1.1482 14.322>}]
 vine_base_a_01 [{shapeType=Capsule, length=1.7873, radius=1.9864, center=<2.3933 0 0>, rotation=<rot 90 0 0>}]
 vine_base_a_02 [{shapeType=Capsule, length=1.7873, radius=1.9864, center=<2.3933 0 0>, rotation=<rot 90 0 0>}]
 vine_base_a_03 [{shapeType=Capsule, length=1.7873, radius=1.9864, center=<2.3933 0 0>, rotation=<rot 90 0 0>}]
 vine_base_a_MSTR [{shapeType=Capsule, length=0.505, radius=0.505, rotation=<rot 0 -90 0>}]
 vine_base_b_01 [{shapeType=Capsule, length=2.0452, radius=1.0802, center=<1.6473 0 0>, rotation=<rot 90 0 0>}]
 vine_base_b_02 [{shapeType=Capsule, length=2.0452, radius=1.0802, center=<1.9554 0 0>, rotation=<rot 90 0 0>}]
 vine_base_b_MSTR [{shapeType=Capsule, length=0.505, radius=0.505, rotation=<rot 0 -90 0>}]
 vine_base_c_01 [{shapeType=Capsule, length=3.1327, radius=2.4379, center=<3.8944 0 0>, rotation=<rot 90 0 0>}]
 vine_base_c_02 [{shapeType=Capsule, length=0.4587, radius=2.4017, center=<2.7465 0 0>, rotation=<rot 90 0 0>}]
 vine_base_c_03 [{shapeType=Capsule, length=0.0488, radius=1.9512, center=<-0.1378 0 -1.6321>}]
 vine_base_c_MSTR [{shapeType=Capsule, length=0.505, radius=0.505, rotation=<rot 0 -90 0>}]
 vine_head_a_01 [{shapeType=Capsule, length=0.2385, radius=1.7615, center=<0 0 -3.153>, rotation=<rot 90 0 0>}]
 vine_head_a_MSTR [{shapeType=Capsule, length=0.505, radius=0.505, rotation=<rot 0 -90 0>}]
 vine_head_b_01 [{shapeType=Capsule, length=0.2385, radius=1.7615, center=<0 0 -3.1911>, rotation=<rot 90 0 0>}]
 vine_head_b_MSTR [{shapeType=Capsule, length=0.505, radius=0.505, rotation=<rot 0 -90 0>}]
 vine_head_c_01 [{shapeType=Capsule, length=0, radius=3.3537, center=<0 0 -3.3885>, rotation=<rot 90 0 0>}]
 vine_head_c_MSTR [{shapeType=Capsule, length=0.505, radius=0.505, rotation=<rot 0 -90 0>}]
 vine_head_d_01 [{shapeType=Capsule, length=0.3876, radius=2.4215, center=<0 0 -2.4405>, rotation=<rot 90 0 0>}]
 vine_head_d_MSTR [{shapeType=Capsule, length=0.505, radius=0.505, rotation=<rot 0 -90 0>}]
 vine_head_e_01 [{shapeType=Capsule, length=1.1764, radius=2.9258, center=<2.5028 0 0>, rotation=<rot 90 0 0>}]
 vine_head_e_02 [{shapeType=Capsule, length=1.1764, radius=2.9258, center=<2.5333 0 0>, rotation=<rot 90 0 0>}]
 vine_head_e_MSTR [{shapeType=Capsule, length=0.505, radius=0.505, rotation=<rot 0 -90 0>}]
 vine_head_f_01 [{shapeType=Capsule, length=1.3518, radius=1.5629, center=<2.0815 0 0>, rotation=<rot 90 0 0>}]
 vine_head_f_02 [{shapeType=Capsule, length=1.9119, radius=1.5629, center=<2.258 0 0>, rotation=<rot 90 0 0>}]
 vine_head_f_MSTR [{shapeType=Capsule, length=0.505, radius=0.505, rotation=<rot 0 -90 0>}]
 vine_head_g_01 [{shapeType=Capsule, length=0.9089, radius=1.8691, center=<1.7102 0 0>, rotation=<rot 90 0 0>}]
 vine_head_g_02 [{shapeType=Capsule, length=0.9089, radius=1.8691, center=<1.7102 0 0>, rotation=<rot 90 0 0>}]
 vine_head_g_MSTR [{shapeType=Capsule, length=0.505, radius=0.505, rotation=<rot 0 -90 0>}]
 vine_head_h_01 [{shapeType=Capsule, length=0.7961, radius=1.7636, center=<1.9158 0 0>, rotation=<rot 90 0 0>}]
 vine_head_h_02 [{shapeType=Capsule, length=0.7961, radius=1.7636, center=<1.9158 0 0>, rotation=<rot 90 0 0>}]
 vine_head_h_MSTR [{shapeType=Capsule, length=0.505, radius=0.505, rotation=<rot 0 -90 0>}]
 vine_shaft_a_01 [{shapeType=Capsule, length=2.2485, radius=1.6307, center=<0 0 -2.3539>, rotation=<rot 90 0 0>}]
 vine_shaft_a_MSTR [{shapeType=Capsule, length=0.505, radius=0.505, rotation=<rot 0 -90 0>}]
 vine_shaft_b_01 [{shapeType=Capsule, length=1.6423, radius=1.78, center=<0 0 -2.8513>, rotation=<rot 90 0 0>}]
 vine_shaft_b_MSTR [{shapeType=Capsule, length=0.505, radius=0.505, rotation=<rot 0 -90 0>}]
}
constraints [
 {
  bone1Name=vine_head_a_MSTR
  bone2Name=Melee_TwoHanded_Handle
  swing1LimitDegrees=45
  swing1Motion=Free
  swing2LimitDegrees=45
  swing2Motion=Free
  twistLimitDegrees=45
  twistMotion=Free
 }
 {
  bone1Name=vine_head_b_MSTR
  bone2Name=Melee_TwoHanded_Handle
  swing1LimitDegrees=45
  swing1Motion=Free
  swing2LimitDegrees=45
  swing2Motion=Free
  twistLimitDegrees=45
  twistMotion=Free
 }
 {
  bone1Name=vine_head_c_MSTR
  bone2Name=Melee_TwoHanded_Handle
  swing1LimitDegrees=45
  swing1Motion=Free
  swing2LimitDegrees=45
  swing2Motion=Free
  twistLimitDegrees=45
  twistMotion=Free
 }
 {
  bone1Name=vine_head_d_MSTR
  bone2Name=Melee_TwoHanded_Handle
  swing1LimitDegrees=45
  swing1Motion=Free
  swing2LimitDegrees=45
  swing2Motion=Free
  twistLimitDegrees=45
  twistMotion=Free
 }
 {
  bone1Name=vine_shaft_b_MSTR
  bone2Name=Melee_TwoHanded_Handle
  swing1LimitDegrees=45
  swing1Motion=Free
  swing2LimitDegrees=45
  swing2Motion=Free
  twistLimitDegrees=45
  twistMotion=Free
 }
 {
  bone1Name=vine_shaft_a_MSTR
  bone2Name=Melee_TwoHanded_Handle
  swing1LimitDegrees=45
  swing1Motion=Free
  swing2LimitDegrees=45
  swing2Motion=Free
  twistLimitDegrees=45
  twistMotion=Free
 }
 {
  bone1Name=vine_head_e_MSTR
  bone2Name=Melee_TwoHanded_Handle
  swing1LimitDegrees=45
  swing1Motion=Free
  swing2LimitDegrees=45
  swing2Motion=Free
  twistLimitDegrees=45
  twistMotion=Free
 }
 {
  bone1Name=vine_head_f_MSTR
  bone2Name=Melee_TwoHanded_Handle
  swing1LimitDegrees=45
  swing1Motion=Free
  swing2LimitDegrees=45
  swing2Motion=Free
  twistLimitDegrees=45
  twistMotion=Free
 }
 {
  bone1Name=vine_head_g_MSTR
  bone2Name=Melee_TwoHanded_Handle
  swing1LimitDegrees=45
  swing1Motion=Free
  swing2LimitDegrees=45
  swing2Motion=Free
  twistLimitDegrees=45
  twistMotion=Free
 }
 {
  bone1Name=vine_head_h_MSTR
  bone2Name=Melee_TwoHanded_Handle
  swing1LimitDegrees=45
  swing1Motion=Free
  swing2LimitDegrees=45
  swing2Motion=Free
  twistLimitDegrees=45
  twistMotion=Free
 }
 {
  bone1Name=vine_base_a_MSTR
  bone2Name=Melee_TwoHanded_Handle
  swing1LimitDegrees=45
  swing1Motion=Free
  swing2LimitDegrees=45
  swing2Motion=Free
  twistLimitDegrees=45
  twistMotion=Free
 }
 {
  bone1Name=vine_base_b_MSTR
  bone2Name=Melee_TwoHanded_Handle
  swing1LimitDegrees=45
  swing1Motion=Free
  swing2LimitDegrees=45
  swing2Motion=Free
  twistLimitDegrees=45
  twistMotion=Free
 }
 {
  bone1Name=vine_base_c_MSTR
  bone2Name=Melee_TwoHanded_Handle
  swing1LimitDegrees=45
  swing1Motion=Free
  swing2LimitDegrees=45
  swing2Motion=Free
  twistLimitDegrees=45
  twistMotion=Free
 }
 {bone1Name=vine_head_a_01, bone2Name=vine_head_a_MSTR, swing1LimitDegrees=50, swing2LimitDegrees=50, twistLimitDegrees=30}
 {bone1Name=vine_head_b_01, bone2Name=vine_head_b_MSTR, swing1LimitDegrees=50, swing2LimitDegrees=50, twistLimitDegrees=30}
 {bone1Name=vine_head_c_01, bone2Name=vine_head_c_MSTR, swing1LimitDegrees=50, swing2LimitDegrees=50, twistLimitDegrees=30}
 {bone1Name=vine_head_d_01, bone2Name=vine_head_d_MSTR, swing1LimitDegrees=50, swing2LimitDegrees=50, twistLimitDegrees=30}
 {bone1Name=vine_shaft_b_01, bone2Name=vine_shaft_b_MSTR, swing1LimitDegrees=80, swing2LimitDegrees=80, twistLimitDegrees=20}
 {bone1Name=vine_shaft_a_01, bone2Name=vine_shaft_a_MSTR, swing1LimitDegrees=80, swing2LimitDegrees=80, twistLimitDegrees=20}
 {
  bone1Name=vine_head_e_01
  bone2Name=vine_head_e_MSTR
  swing1LimitDegrees=80
  swing1Motion=Free
  swing2LimitDegrees=80
  swing2Motion=Free
  twistLimitDegrees=20
 }
 {
  bone1Name=vine_head_e_02
  bone2Name=vine_head_e_01
  swing1LimitDegrees=80
  swing1Motion=Free
  swing2LimitDegrees=80
  swing2Motion=Free
  twistLimitDegrees=20
 }
 {
  bone1Name=vine_head_f_01
  bone2Name=vine_head_f_MSTR
  swing1LimitDegrees=80
  swing1Motion=Free
  swing2LimitDegrees=80
  swing2Motion=Free
  twistLimitDegrees=20
 }
 {
  bone1Name=vine_head_f_02
  bone2Name=vine_head_f_01
  swing1LimitDegrees=80
  swing1Motion=Free
  swing2LimitDegrees=80
  swing2Motion=Free
  twistLimitDegrees=20
 }
 {
  bone1Name=vine_head_g_01
  bone2Name=vine_head_g_MSTR
  swing1LimitDegrees=80
  swing1Motion=Free
  swing2LimitDegrees=80
  swing2Motion=Free
  twistLimitDegrees=20
 }
 {
  bone1Name=vine_head_g_02
  bone2Name=vine_head_g_01
  swing1LimitDegrees=80
  swing1Motion=Free
  swing2LimitDegrees=80
  swing2Motion=Free
  twistLimitDegrees=20
 }
 {
  bone1Name=vine_head_h_01
  bone2Name=vine_head_h_MSTR
  swing1LimitDegrees=80
  swing1Motion=Free
  swing2LimitDegrees=80
  swing2Motion=Free
  twistLimitDegrees=20
 }
 {
  bone1Name=vine_head_h_02
  bone2Name=vine_head_h_01
  swing1LimitDegrees=80
  swing1Motion=Free
  swing2LimitDegrees=80
  swing2Motion=Free
  twistLimitDegrees=20
 }
 {bone1Name=vine_base_a_01, bone2Name=vine_base_a_MSTR, swing1LimitDegrees=80, swing2LimitDegrees=80, twistLimitDegrees=20}
 {bone1Name=vine_base_a_02, bone2Name=vine_base_a_01, swing1LimitDegrees=80, swing2LimitDegrees=80, twistLimitDegrees=20}
 {bone1Name=vine_base_a_03, bone2Name=vine_base_a_02, swing1LimitDegrees=80, swing2LimitDegrees=80, twistLimitDegrees=20}
 {bone1Name=vine_base_b_01, bone2Name=vine_base_b_MSTR, swing1LimitDegrees=80, swing2LimitDegrees=80, twistLimitDegrees=20}
 {bone1Name=vine_base_b_02, bone2Name=vine_base_b_01, swing1LimitDegrees=80, swing2LimitDegrees=80, twistLimitDegrees=20}
 {bone1Name=vine_base_c_01, bone2Name=vine_base_c_MSTR, swing1LimitDegrees=80, swing2LimitDegrees=80, twistLimitDegrees=20}
 {bone1Name=vine_base_c_02, bone2Name=vine_base_c_01, swing1LimitDegrees=80, swing2LimitDegrees=80, twistLimitDegrees=20}
 {bone1Name=vine_base_c_03, bone2Name=vine_base_c_02, swing1LimitDegrees=80, swing2LimitDegrees=80, twistLimitDegrees=20}
]
